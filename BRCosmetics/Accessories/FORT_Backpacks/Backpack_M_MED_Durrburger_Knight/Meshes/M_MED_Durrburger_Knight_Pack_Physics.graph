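# Physics asset for the Durrburger Knight backpack, giving its ketchup, pickle and tongue bones collision bodies and joint limits for secondary motion.
# One large capsule covers the pack bone, with small capsules on each dyn_ joint.
# Ketchup and tongue form chains from the pack, while pickle joints 1 and 2 both attach directly to it and joint 3 follows 2.
# Ketchup and pickle constraints lock twist; tongue constraints allow limited twist.
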PhysicsAsset /BRCosmetics/Accessories/FORT_Backpacks/Backpack_M_MED_Durrburger_Knight/Meshes/M_MED_Durrburger_Knight_Pack_Physics
bodies {
 dyn_bigKetchup_jnt_1 [{shapeType=Capsule, length=1, radius=2.5, center=<1 0 0>, rotation=<rot 90 0 0>}]
 dyn_bigKetchup_jnt_2 [{shapeType=Capsule, length=1, radius=2.5, center=<0.9094 0 0>, rotation=<rot 90 0 0>}]
 dyn_pickle_jnt_1 [{shapeType=Capsule, length=0.505, radius=2, center=<0.3698 0 0>, rotation=<rot 90 0 0>}]
 dyn_pickle_jnt_2 [{shapeType=Capsule, length=0.505, radius=2, center=<0.2154 0 0>, rotation=<rot 90 0 0>}]
 dyn_pickle_jnt_3 [{shapeType=Capsule, length=0.505, radius=2, center=<0.8987 0 0>, rotation=<rot 90 0 0>}]
 dyn_tongue_jnt_1 [{shapeType=Capsule, length=3, radius=4, center=<2.4835 0 0>, rotation=<rot 90 0 0>}]
 dyn_tongue_jnt_2 [{shapeType=Capsule, length=3, radius=4, center=<2.0049 0 0>, rotation=<rot 90 0 0>}]
 dyn_tongue_jnt_3 [{shapeType=Capsule, length=3, radius=4, center=<2.0049 0 0>, rotation=<rot 90 0 0>}]
 pack [{shapeType=Capsule, length=50, radius=7, center=<-0.6148 -10.6661 -4.1892>, rotation=<rot 28.4856 -180 -180>}]
}
constraints [
 {
  bone1Name=dyn_bigKetchup_jnt_2
  bone2Name=dyn_bigKetchup_jnt_1
  swing1LimitDegrees=45
  swing2LimitDegrees=45
  twistLimitDegrees=45
  twistMotion=Locked
 }
 {bone1Name=dyn_tongue_jnt_2, bone2Name=dyn_tongue_jnt_1, swing1LimitDegrees=45, swing2LimitDegrees=35, twistLimitDegrees=20}
 {bone1Name=dyn_tongue_jnt_3, bone2Name=dyn_tongue_jnt_2, swing1LimitDegrees=45, swing2LimitDegrees=35, twistLimitDegrees=20}
 {bone1Name=dyn_pickle_jnt_2, bone2Name=pack, swing1LimitDegrees=45, swing2LimitDegrees=45, twistLimitDegrees=45, twistMotion=Locked}
 {
  bone1Name=dyn_pickle_jnt_3
  bone2Name=dyn_pickle_jnt_2
  swing1LimitDegrees=45
  swing2LimitDegrees=45
  twistLimitDegrees=45
  twistMotion=Locked
 }
 {bone1Name=dyn_pickle_jnt_1, bone2Name=pack, swing1LimitDegrees=45, swing2LimitDegrees=45, twistLimitDegrees=45, twistMotion=Locked}
 {bone1Name=dyn_bigKetchup_jnt_1, bone2Name=pack, swing1LimitDegrees=45, swing2LimitDegrees=45, twistLimitDegrees=45, twistMotion=Locked}
 {bone1Name=dyn_tongue_jnt_1, bone2Name=pack, swing1LimitDegrees=45, swing2LimitDegrees=30, twistLimitDegrees=15}
]
