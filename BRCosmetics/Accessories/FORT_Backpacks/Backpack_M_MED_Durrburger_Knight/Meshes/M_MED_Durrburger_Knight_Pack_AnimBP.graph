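# Animation Blueprint for the Durrburger Knight backpack, adding secondary physics motion to the pack.
# AnimDynamics nodes slide the burger and two pickle bones along one constrained axis relative to the pack bone.
# Between them, a RigidBody node simulates the pack physics asset, switching gravity override by front-end and on-ground state.
# A final additive rotation tilts the pack while skydiving; the event graph has no real logic.
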AnimBlueprint /BRCosmetics/Accessories/FORT_Backpacks/Backpack_M_MED_Durrburger_Knight/Meshes/M_MED_Durrburger_Knight_Pack_AnimBP
target_skeleton=&/M_MED_Durrburger_Knight_Pack_Skeleton
graphs [
 AnimGraph {nodes [
   n1 = Anim Dynamics - Bone: dyn_burger_slidey_jnt(ComponentPose=Local To Component(LocalPose=Input Pose()), Alpha=1, ExternalForce="0,0,0") {
    bGravityOverrideInSimSpace=true
    bOverrideLinearDamping=true
    gravityScale=0.999
    linearDampingOverride=1
    simulationSpace=RootRelative
    alphaBoolBlend {blendOption=Linear}
    boundBone {boneName=dyn_burger_slidey_jnt}
    componentAppliedLinearAccClamp=<1000 4000 4000>
    componentLinearAccScale=<1.02 1.02 1.02>
    physicsBodyDefinitions [{
      sphereCollisionRadius=0
      boundBone {boneName=dyn_burger_slidey_jnt}
      boxExtents=<1 1 1>
      constraintSetup {angularLimitsMax=<0 0 0>, angularLimitsMin=<0 0 0>, angularTarget=<0 0 0>, linearAxesMax=<1.5 0 0>, linearAxesMin=<-1 0 0>}
      localJointOffset=<0 0 0>
     }]
    relativeSpaceBone {boneName=pack}
   }
   n2 = NOT Boolean(A=Get bIsOnGround() -> (bIsOnGround:bool))
   n3 = RigidBody!(
    ComponentPose=@n1
    Alpha
    OverrideWorldGravity:Vector=SelectVector(A="0,0,-900", B=SelectVector(A="0,-100,-100", B="0, 0,-980", bPickA=@n2), bPickA=Get bIsInFrontEnd())
    ExternalForce
   ) -> (Pose) {
    bClampLinearTranslationLimitToRefPose=true
    bOverrideWorldGravity=true
    overridePhysicsAsset=&/M_MED_Durrburger_Knight_Pack_Physics
    simulationSpace=BaseBoneSpace
    alphaBoolBlend {blendOption=Linear}
    baseBoneRef {boneName=root}
    componentAppliedLinearAccClamp=<5000 5000 5000>
    componentLinearAccScale=<0.25 0.25 0.25>
    componentLinearVelScale=<0.3 0.3 0.3>
    overrideWorldGravity {y=-200, z=-200}
   }
   n4 = Anim Dynamics - Bone: dyn_pickle1_slidey_jnt(ComponentPose=@n3, Alpha=1, ExternalForce="0,0,0") {
    bGravityOverrideInSimSpace=true
    bOverrideLinearDamping=true
    gravityScale=0.999
    linearDampingOverride=1
    simulationSpace=RootRelative
    alphaBoolBlend {blendOption=Linear}
    boundBone {boneName=dyn_pickle1_slidey_jnt}
    componentAppliedLinearAccClamp=<1000 4000 4000>
    componentLinearAccScale=<1.02 1.02 1.02>
    physicsBodyDefinitions [{
      sphereCollisionRadius=0
      boundBone {boneName=dyn_pickle1_slidey_jnt}
      boxExtents=<1 1 1>
      constraintSetup {angularLimitsMax=<0 0 0>, angularLimitsMin=<0 0 0>, angularTarget=<0 0 0>, linearAxesMax=<1 0 0>, linearAxesMin=<-0.5 0 0>}
      localJointOffset=<0 0 0>
     }]
    relativeSpaceBone {boneName=pack}
   }
   n5 = Anim Dynamics - Bone: dyn_pickle2_slidey_jnt(ComponentPose=@n4, Alpha=1, ExternalForce="0,0,0") {
    bGravityOverrideInSimSpace=true
    bOverrideLinearDamping=true
    linearDampingOverride=1
    simulationSpace=RootRelative
    alphaBoolBlend {blendOption=Linear}
    boundBone {boneName=dyn_pickle2_slidey_jnt}
    componentAppliedLinearAccClamp=<2000 4000 4000>
    componentLinearAccScale=<1.03 1.03 1.03>
    physicsBodyDefinitions [{
      sphereCollisionRadius=0
      boundBone {boneName=dyn_pickle2_slidey_jnt}
      boxExtents=<1 1 1>
      constraintSetup {angularLimitsMax=<0 0 0>, angularLimitsMin=<0 0 0>, angularTarget=<0 0 0>, linearAxesMax=<0.9 0 0>, linearAxesMin=<-0.4 0 0>}
      localJointOffset=<0 0 0>
     }]
    relativeSpaceBone {boneName=pack}
   }
   n6 = "Transform (Modify) Bone - Bone: pack"(ComponentPose=@n5, Alpha="To Float (Boolean)"(InBool=Get bIsSkyDiving()), Translation="0,0,0", Rotation="-10,0,5") {rotationMode=BMM_Additive, alphaBoolBlend {blendOption=Linear}, boneToModify {boneName=pack}, rotation {pitch=-10, roll=5}}
   Output Pose(Result=Component To Local(ComponentPose=@n6)) {layerGroup=None}
  ]}
 EventGraph {nodes [
   TryGetPawnOwner()
   Event BlueprintUpdateAnimation()
  ]}
]
variables []
